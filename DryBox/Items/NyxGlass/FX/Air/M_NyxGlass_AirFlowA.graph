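# Translucent, unlit, two-sided Niagara mesh-particle material for the NyxGlass item's air flow effect, drawing streaks broken up by scrolling noise.
# Two panning grayscale noise textures distort the mesh UVs, and the distorted V is shaped into a mask, then multiplied by an edge mask on undistorted V.
# A facing-based silhouette fade varies its exponent with the camera's alignment to the instance X axis, softening end-on views.
# Particle color and alpha tint and fade the result, with depth fade softening intersections.
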Material /DryBox/Items/NyxGlass/FX/Air/M_NyxGlass_AirFlowA
graph {
 properties {
  bUsedWithNiagaraMeshParticles=true
  blendMode=BLEND_Translucent
  pixelDepthOffsetMode=PDOM_Legacy
  shadingModel=MSM_Unlit
  twoSided=true
  usedShadingModels=Unlit
 }
 nodes [
  n1 = TextureObjectParameter() {
   group=SecTexture
   parameterName=SecTexture_Texture
   samplerType=SAMPLERTYPE_LinearGrayscale
   sortPriority=0
   texture=/Game/Effects/Fort_Effects/Emotes/UltraLight/Texture/T_Plasma_A_B_Greyscale
  }
  n2 = VectorParameter() {group=SecTexture, parameterName=SecTexture_UVControl, channelNames {a=VPan, b=UPan, g=VScale, r=UScale}, defaultValue=<rgba 1 0.1 0 10>}
  n3 = VectorParameter() {
   group=SecTexture
   parameterName=SecTexture_NoiseControl
   sortPriority=60
   channelNames {a=VStrength, b=UStrength, g=VOffset, r=UOffset}
   defaultValue=<rgba -0.5 -0.5 0 0.25>
  }
  n4 = MaterialFunctionCall!(
   NoiseValue=MaterialFunctionCall(Texture=@n1, UVControl_RGB=@n2, UVControl_A=@n2) {materialFunction=&FXCommon_TextureSample}.RG
   Control_RGB=@n3
   Control_A=@n3
  ) {materialFunction=/Game/Effects/Niagara/MasterMaterials/Functions/Common/FXCommon_NoiseControl}
  n5 = TextureObjectParameter() {
   group=MainTexture
   parameterName=MainTexture_Texture
   samplerType=SAMPLERTYPE_LinearGrayscale
   sortPriority=0
   texture=/Game/Effects/Fort_Effects/Emotes/UltraLight/Texture/T_DirectionalNoise_Greyscale
  }
  n6 = VectorParameter() {group=MainTexture, parameterName=MainTexture_UVControl, channelNames {a=VPan, b=UPan, g=VScale, r=UScale}, defaultValue=<rgba 1 0.05 0 20>}
  n7 = VectorParameter() {
   group=MainTexture
   parameterName=MainTexture_NoiseControl
   sortPriority=60
   channelNames {a=VStrength, b=UStrength, g=VOffset, r=UOffset}
   defaultValue=<rgba -0.5 -0.5 0 0.3>
  }
  n8 = Add(
   A=TextureCoordinate()
   B=MaterialFunctionCall(
    NoiseValue=MaterialFunctionCall(Texture=@n5, UVControl_RGB=@n6, UVControl_A=@n6) {materialFunction=&FXCommon_TextureSample}.RG
    Control_RGB=@n7
    Control_A=@n7
   ) {materialFunction=&FXCommon_NoiseControl}.NoiseValue
  )
  n9 = VectorParameter() {group=EdgeMask, parameterName=EdgeMask_EdgeMask_Control, channelNames {a="", b=Power, g=Multiply, r=Offset}, defaultValue=<rgba 0 1.25 4 1>}
  n10 = Add(
   A=MaterialFunctionCall(Float2=Add(A=@n4.NoiseValue, B=@n8)) {materialFunction=&BreakOutFloat2Components}.G
   B=@n9.Offset
  )
  n11 = Power(Base=Saturate(None=Multiply(A=@n10, B=@n9.Multiply)), Exp=@n9.Power)
  n12 = TextureCoordinate()
  n13 = VectorParameter() {
   group=EdgeMask
   parameterName=EdgeMask_EdgeMask_InvertControl
   channelNames {a="", b=Power, g=Multiply, r=Offset}
   defaultValue=<rgba 0 15 2 1>
  }
  n14 = DotProduct(A=CameraVectorWS(), B=PixelNormalWS())
  n15 = Power(Base=OneMinus(None=Saturate(None=@n14)), Exp=ScalarParameter() {defaultValue=1, parameterName=F2})
  n16 = ScalarParameter() {defaultValue=15, parameterName=F}
  n17 = Transform(None=Constant3Vector() {constant=<rgba -1 0 0 0>}) {transformSourceType=TRANSFORMSOURCE_Instance}
  n18 = Abs(None=DotProduct(
    A=Normalize(VectorInput=@n17)
    B=MaterialFunctionCall() {materialFunction=&CameraDirectionVector}.Result
   ))
  n19 = Multiply(
   A=MaterialFunctionCall!(
    UorV=MaterialFunctionCall(Float2=@n12) {materialFunction=&BreakOutFloat2Components}.G
    Invert_EdgeControl=@n13
    NoInvert_EdgeControl=Constant4Vector() {constant=<rgba 0 1 0 0>}.RGBA
   ) {materialFunction=/Game/Effects/Niagara/MasterMaterials/Functions/Common/FXCommon_EdgeMask_Ex}.Mask
   B=Power(Base=OneMinus(None=@n15), Exp=LinearInterpolate(A=@n16, Alpha=@n18))
  )
  n20 = ParticleColor()
  DepthFade(Opacity=Multiply(A=Multiply(A=@n11, B=@n19), B=@n20.A), FadeDistance=ScalarParameter() {defaultValue=5, parameterName=DF})
  Fresnel() {baseReflectFraction=0, exponent=1}
  Multiply(
   A=MaterialFunctionCall(Emissive=VectorParameter() {parameterName=Color, defaultValue=<rgba 1 1 1 1>}.RGB) {materialFunction=&MF_FX_EmissiveAdjustments}.Result
   B=@n20.RGB
  )
 ]
}
