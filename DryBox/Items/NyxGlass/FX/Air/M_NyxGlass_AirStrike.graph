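# Unlit, two-sided translucent material for Niagara mesh particles in the NyxGlass AirStrike effect, drawing noise-distorted streaks.
# Emissive switches between two color parameters with a hard step at the V midpoint, tinted by particle color.
# Opacity multiplies a U-based edge mask, particle alpha and a ribbon texture warped by panning Perlin noise.
# A power curve on U blends two ribbon UV controls, sharply increasing V tiling toward one end.
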Material /DryBox/Items/NyxGlass/FX/Air/M_NyxGlass_AirStrike
graph {
 properties {
  bUsedWithNiagaraMeshParticles=true
  blendMode=BLEND_Translucent
  pixelDepthOffsetMode=PDOM_Legacy
  shadingModel=MSM_Unlit
  twoSided=true
  usedShadingModels=Unlit
 }
 nodes [
  n1 = VectorParameter() {parameterName=Color, defaultValue=<rgba 0.15 0.15 0.15 1>}
  n2 = VectorParameter() {parameterName=Color2, defaultValue=<rgba 0.5 0.5 0.5 1>}
  n3 = MaterialFunctionCall(Float2=TextureCoordinate()) {materialFunction=&BreakOutFloat2Components}
  n4 = Floor(None=Multiply(A=@n3.G) {constB=2})
  n5 = ParticleColor()
  MaterialFunctionCall(Emissive=Multiply(A=LinearInterpolate(A=@n1.RGB, B=@n2.RGB, Alpha=@n4), B=@n5.RGB)) {materialFunction=&MF_FX_EmissiveAdjustments}
  n6 = VectorParameter() {group=Opacity, parameterName=Opacity_EdgeMask_InvertControl, channelNames {a="", b=Power, g=Multiply, r=Offset}, defaultValue=<rgba 0 4 5 1>}
  n7 = VectorParameter() {group=Opacity, parameterName=Opacity_EdgeMask_Control, channelNames {a="", b=Power, g=Multiply, r=Offset}, defaultValue=<rgba 0 2 2 1>}
  n8 = TextureObjectParameter() {
   group=Ribbon
   parameterName=Ribbon_Texture
   samplerType=SAMPLERTYPE_Grayscale
   sortPriority=0
   texture=/Game/Gadgets/Assets/VinderTech_GliderChute/Glider_Thrive_Female/FX/Texture/T_RibbonBaseShape4
  }
  n9 = TextureObjectParameter() {
   group=MainT
   parameterName=MainT_Texture
   samplerType=SAMPLERTYPE_LinearGrayscale
   sortPriority=0
   texture=&T_SimplePerlin_Greyscale
  }
  n10 = VectorParameter() {group=MainT, parameterName=MainT_UVControl, channelNames {a=VPan, b=UPan, g=VScale, r=UScale}, defaultValue=<rgba 0.5 0.5 2 0.1>}
  n11 = StaticBool() {value=true}
  n12 = VectorParameter() {
   group=MainT
   parameterName=MainT_NoiseControl
   sortPriority=60
   channelNames {a=VStrength, b=UStrength, g=VOffset, r=UOffset}
   defaultValue=<rgba 0.0309 -0.5 1 0.2>
  }
  n13 = MaterialFunctionCall!(
   NoiseValue=MaterialFunctionCall!(
    Texture=@n9
    AdditionalUVOffset
    UVControl_RGB=@n10
    UVControl_A=@n10
    "[Override]TimeInput"
    "[Override]ScaleUVByCenter"=@n11
    "[Override]UVInput"
   ) {materialFunction=/Game/Effects/Niagara/MasterMaterials/Functions/Common/FXCommon_TextureSample_V2}.RG
   Control_RGB=@n12
   Control_A=@n12
  ) {materialFunction=/Game/Effects/Niagara/MasterMaterials/Functions/Common/FXCommon_NoiseControl}
  n14 = VectorParameter() {group=Ribbon, parameterName=Ribbon_UVControl, channelNames {a=VPan, b=UPan, g=VScale, r=UScale}, defaultValue=<rgba 1 2 0 0>}
  n15 = VectorParameter() {group=Ribbon, parameterName=Ribbon_UVControl2, channelNames {a=VPan, b=UPan, g=VScale, r=UScale}, defaultValue=<rgba 1 25 0 0>}
  n16 = TextureCoordinate()
  n17 = LinearInterpolate(
   A=@n14
   B=@n15
   Alpha=Power(
    Base=MaterialFunctionCall(Float2=@n16) {materialFunction=&BreakOutFloat2Components}.R
    Exp=ScalarParameter() {defaultValue=10, parameterName=aw}
   )
  )
  Multiply(
   A=MaterialFunctionCall!(UorV=@n3.R, Invert_EdgeControl=@n6, NoInvert_EdgeControl=@n7) {materialFunction=/Game/Effects/Niagara/MasterMaterials/Functions/Common/FXCommon_EdgeMask_Ex}.Mask
   B=Multiply(
    A=@n5.A
    B=MaterialFunctionCall(Texture=@n8, AdditionalUVOffset=@n13.NoiseValue, UVControl_RGB=@n17, UVControl_A=@n14.VPan, "[Override]ScaleUVByCenter"=@n11) {materialFunction=&FXCommon_TextureSample_V2}.R
   )
  )
 ]
}
