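# Additive, unlit, two-sided refraction material for Niagara mesh particles in the NyxGlass air effect.
# An index-of-refraction strength parameter fades to neutral with camera distance, is masked by a powered sine across UV X, and scales with particle alpha.
# SplineThicken supplies a constant-width offset, which is added to world position to build a normalized world-space direction radiating from the particle center.
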Material /DryBox/Items/NyxGlass/FX/Air/M_NyxGlass_Refraction
graph {
 properties {
  bEnableResponsiveAA=true
  bTangentSpaceNormal=false
  bUsedWithNiagaraMeshParticles=true
  blendMode=BLEND_Additive
  pixelDepthOffsetMode=PDOM_Legacy
  refractionMethod=RM_IndexOfRefraction
  shadingModel=MSM_Unlit
  twoSided=true
  usedShadingModels=Unlit
 }
 nodes [
  n1 = ScalarParameter() {defaultValue=0.9, parameterName=RefStr}
  n2 = Distance(A=WorldPosition().XYZ, B=CameraPositionWS())
  n3 = Saturate(None=Divide(A=Subtract(A=@n2) {constB=2048}) {constB=4096})
  n4 = Multiply(A=MaterialFunctionCall(Float2=TextureCoordinate()) {materialFunction=&BreakOutFloat2Components}.R) {constB=0.5}
  n5 = Multiply(
   A=ScalarParameter() {defaultValue=1, parameterName=awtatatat}
   B=Power(Base=Sine(None=@n4), Exp=ScalarParameter() {defaultValue=10, parameterName=afwafw})
  )
  LinearInterpolate(B=LinearInterpolate(B=LinearInterpolate(A=@n1, Alpha=@n3), Alpha=Saturate(None=@n5)) {constA=1}, Alpha=ParticleColor().A) {constA=1}
  n6 = WorldPosition()
  n7 = ScalarParameter() {defaultValue=400, parameterName=Width}
  n8 = MaterialFunctionCall(WidthBase=@n7, WidthTip=@n7, FlattenPixelNormal=StaticBool() {value=true}) {materialFunction=&SplineThicken}
  Normalize(VectorInput=Subtract(A=Add(A=@n6.XYZ, B=@n8.WorldPositionOffset), B=ParticlePositionWS()))
 ]
}
