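# Niagara system for the NyxGlass item's air effects, layering gas-like ribbons and swirling meshes on a character.
# Ribbon emitters trail five skeletal mesh bones, the hands and the owner's position (Steam).
# SandWarp004 spawns spinning, scaling meshes at one bone, gated by aim-attack, attack and owner-speed checks.
# Counter-driven custom modules correct hand and Steam ribbon heads, while time of day adjusts bone-ribbon opacity and mesh material parameters.
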NiagaraSystem /DryBox/Items/NyxGlass/FX/Air/NS_GasParticles_NyxGlass
dependencies {
 usedDataInterfaces [
  &NiagaraDataInterfaceArrayFloat
  &NiagaraDataInterfaceArrayInt32
  &NiagaraDataInterfaceArrayFloat3
 ]
 usedDynamicInputs [&Lerp_Float]
 usedModules [
  &SystemState
  &self:SystemSpawnScript.SystemScriptSource.SystemScriptGraph.NiagaraNodeAssignment_0.SetVariables_#1
 ]
}
emitters {
 GrindRail_Main {
  emitterUpdateScript [
   &EmitterState()
   &FN_SpawnRibbon_v2(
    Spawn Mode=::NewEnumerator0
    SpawnRate:NiagaraFloat=&FloatFromCurve(FloatCurve=DataInterface() {propertyValues {Curve {keys [<at 0 0>,<at 0.05 15>,<at 1 40>]}}}, CurveIndex=0.5, Scale Curve=6)
   )
  ]
  particleSpawnScript [
   &FN_SetSpawnGroup(Number of Spawn Groups=5)
   &InitializeRibbon(
    Color=<rgba 1.6 1.6667 2 0.1>
    Lifetime=&SelectFloatFromArray(Direct Array Index=$Particles.SpawnGroup, Float Selection Array=DataInterface() {propertyValues {FloatData [0.15,0.2,0.25,0.15,0.15]}})
    Module.Write Color=true
    Module.Write Lifetime=true
    Module.Write Mass=true
    Module.Write Material Random=true
    Module.Write Position=false
    Module.Write RibbonTwist=true
    Module.Write RibbonWidth=true
    Ribbon Twist:NiagaraFloat=&SelectFloatFromArray(Direct Array Index=$Particles.SpawnGroup, Float Selection Array=DataInterface() {propertyValues {FloatData [0.5,0.25,1,1,-0.75]}})
    Ribbon Width=&SelectFloatFromArray(Direct Array Index=$Particles.SpawnGroup, Float Selection Array=DataInterface() {propertyValues {FloatData [80,30,50,35,50]}})
   )
   /SetVariables(Particles.RibbonDistanceFromStart:NiagaraFloat=$Emitter.DistanceTraveled) {isSetParameters=true}
   &SkeletalMeshLocation(
    Bone / Socket Index=&SelectIntFromArray(Direct Array Index=$Particles.SpawnGroup, Int Selection Array=DataInterface() {propertyValues {IntData [7,8,9,15,16]}})
    Bone Sampling Mode=::NewEnumerator4
    Interpolate Mesh Sampling=false
    MaskSpawnGroup=false
    Orientation Sampling=::NewEnumerator1
    Sampled Position Offset=&SelectVectorFromArray(
     Direct Array Index=$Particles.SpawnGroup
     Vector Selection Array=DataInterface() {propertyValues {FloatData [<5 -5 -25>, <0 0 10>, <15 10 15>, <0 0 0>, <0 0 0>]}}
    )
    Sampled Position Offset Coordinate Space:ENiagaraSkelMeshTransforms=::NewEnumerator2
    Skeletal Mesh=$User.SkeletalMesh
   )
  ]
  particleUpdateScript [
   &ParticleState()
   &SolveForcesAndVelocity(Clamp Velocity=false, Limit Acceleration=false)
   &ScaleColor(
    Scale Alpha=&FNTimeOfDayScalar(Night Scale=0.5, Morning Scale=0.35, Evening Scale=0.4)
    ScaleA=true
    ScaleRGB=true
   )
  ]
  rendererClasses [&NiagaraRibbonRendererProperties]
  renderers [{rendererClass=&NiagaraRibbonRendererProperties}]
 }
 HandTrail {
  emitterUpdateScript [
   &EmitterState(Life Cycle Mode=::NewEnumerator1, Loop Behavior=::NewEnumerator1, Loop Delay=0.3, Loop Duration Mode=::NewEnumerator1, UseLoopDelay=true)
   &FN_SpawnRibbon_v2(SpawnRate=30)
  ]
  particleSpawnScript [
   &FN_SetSpawnGroup(Number of Spawn Groups=1)
   &InitializeRibbon(Color=<rgba 0.581 0.6207 0.7 0.07>, Lifetime=0.15, Ribbon Width=39, UsePositionOffset=false, Write Material Random=true)
   &SkeletalMeshLocation(
    Bone / Socket Index=$Particles.SpawnGroup
    Bone Sampling Mode=::NewEnumerator4
    MaskSpawnGroup=false
    Skeletal Mesh=DataInterface() {propertyValues {
      DefaultMesh=null
      ExcludeBoneName=root
      OptionalComponentClass=null
      PreviewMesh=&M_Med_Soldier_04_Preview
      SkinningMode=SkinOnTheFly
      SoftSourceActor=null
      SourceMode=Default
      UvSetIndex=0
      WholeMeshLOD=-1
      bExcludeBone=false
      bReadDeformedGeometry=true
      bRequireCurrentFrameData=true
      ComponentTags []
      FilteredBones [hand_r,hand_l]
      FilteredSockets []
      MeshUserParameter {parameter {name=None, typeDefHandle {registeredTypeIndex=1}, varData []}}
      SamplingRegions []
     }}
   )
   &DynamicMaterialParameters(
    Index 0 Param 1=1
    Index 0 Param 2=0.5
    Index 0 Param 3=2
    Index 0 Param 4=1
    Module.Param0WriteEnabled=true
    Module.Param0WriteEnabled1=true
    Module.Param0WriteEnabled2=true
    Module.Param1WriteEnabled=true
    Module.Param1WriteEnabled1=true
    Module.Param1WriteEnabled2=true
    Module.Param2WriteEnabled=true
    Module.Param2WriteEnabled1=true
    Module.Param2WriteEnabled2=true
    Module.Param3WriteEnabled=true
    Module.Param3WriteEnabled1=true
    Module.Param3WriteEnabled2=true
    Write Parameter Index 1=true
    Write Parameter Index 2=true
   )
   &self:ResolveRibbonHeadissuePre(New Simple Counter:NiagaraDataInterfaceSimpleCounter=$User.RibbonCounter1) {module=ResolveRibbonHeadissuePre}
  ]
  particleUpdateScript [
   &ParticleState()
   &ScaleColor(
    Scale Alpha=&FloatFromCurve(
     FloatCurve=DataInterface() {propertyValues {Curve {keys [
         {interpMode=RCIM_Cubic, leaveTangent=13.9561, tangentMode=RCTM_Break, time=0, value=0}
         {arriveTangent=2.4263, interpMode=RCIM_Cubic, tangentMode=RCTM_Break, time=0.1, value=1}
         {arriveTangent=0.265, interpMode=RCIM_Cubic, tangentMode=RCTM_Break, time=1, value=0}
        ]}}}
     Scale Curve=$System.DistanceNearFade
    )
    ScaleA=true
    ScaleRGB=true
   )
   &ScaleRibbonWidth(Ribbon Width Scale=&FloatFromCurve(FloatCurve=DataInterface() {propertyValues {Curve {keys [<at 0 1>,<at 1 0.35>]}}}))
   &ScaleRibbonWidth(Ribbon Width Scale=&ClampFloat(Float=&FN_LOD_Distance_Remap(Start Distance=0, End Distance=2000, Minimum Output=0.5, Maximum Output=1))) {module=ScaleRibbonWidth001}
   &SkeletalMeshLocation(
    Bone / Socket Index=$Particles.SpawnGroup
    Bone Sampling Mode=::NewEnumerator4
    MaskSpawnGroup=false
    Position Sampling=::NewEnumerator2
    Skeletal Mesh=DataInterface() {propertyValues {
      DefaultMesh=null
      ExcludeBoneName=root
      OptionalComponentClass=null
      PreviewMesh=&M_Med_Soldier_04_Preview
      SkinningMode=SkinOnTheFly
      SoftSourceActor=null
      SourceMode=Default
      UvSetIndex=0
      WholeMeshLOD=-1
      bExcludeBone=false
      bReadDeformedGeometry=true
      bRequireCurrentFrameData=true
      ComponentTags []
      FilteredBones [hand_r,hand_l]
      FilteredSockets []
      MeshUserParameter {parameter {name=None, typeDefHandle {registeredTypeIndex=1}, varData []}}
      SamplingRegions []
     }}
   ) {module=SkeletalMeshLocation001}
   &self:ResolveRibbonheadIssue(
    IDCustom:NiagaraInt32=$Particles.IDCustom
    InputHeadPos:NiagaraPosition=$Particles.SkeletalMeshLocation001.SampledPosition
    New Simple Counter:NiagaraDataInterfaceSimpleCounter=$User.RibbonCounter1
   ) {module=ResolveRibbonheadIssue}
  ]
  rendererClasses [&NiagaraRibbonRendererProperties]
  renderers [{rendererClass=&NiagaraRibbonRendererProperties}]
 }
 HandTrail001 {
  emitterUpdateScript [
   &EmitterState(Life Cycle Mode=::NewEnumerator1, Loop Behavior=::NewEnumerator1, Loop Delay=0.3, Loop Duration Mode=::NewEnumerator1, UseLoopDelay=true)
   &FN_SpawnRibbon_v2(SpawnRate=30)
  ]
  particleSpawnScript [
   &FN_SetSpawnGroup(Number of Spawn Groups=1)
   &InitializeRibbon(Color=<rgba 0.581 0.6207 0.7 0.07>, Lifetime=0.15, Ribbon Width=39, UsePositionOffset=false, Write Material Random=true)
   &SkeletalMeshLocation(
    Bone / Socket Index=$Particles.SpawnGroup
    Bone Sampling Mode=::NewEnumerator4
    Interpolate Mesh Sampling=false
    MaskSpawnGroup=false
    Orientation Sampling=::NewEnumerator1
    Skeletal Mesh=DataInterface() {propertyValues {
      DefaultMesh=null
      ExcludeBoneName=root
      OptionalComponentClass=null
      PreviewMesh=&M_Med_Soldier_04_Preview
      SkinningMode=SkinOnTheFly
      SoftSourceActor=null
      SourceMode=Default
      UvSetIndex=0
      WholeMeshLOD=-1
      bExcludeBone=false
      bReadDeformedGeometry=true
      bRequireCurrentFrameData=true
      ComponentTags []
      FilteredBones [hand_l,hand_l]
      FilteredSockets []
      MeshUserParameter {parameter {name=None, typeDefHandle {registeredTypeIndex=1}, varData []}}
      SamplingRegions []
     }}
   )
   &DynamicMaterialParameters(
    Index 0 Param 1=1
    Index 0 Param 2=0.5
    Index 0 Param 3=2
    Index 0 Param 4=1
    Module.Param0WriteEnabled=true
    Module.Param0WriteEnabled1=true
    Module.Param0WriteEnabled2=true
    Module.Param1WriteEnabled=true
    Module.Param1WriteEnabled1=true
    Module.Param1WriteEnabled2=true
    Module.Param2WriteEnabled=true
    Module.Param2WriteEnabled1=true
    Module.Param2WriteEnabled2=true
    Module.Param3WriteEnabled=true
    Module.Param3WriteEnabled1=true
    Module.Param3WriteEnabled2=true
    Write Parameter Index 1=true
    Write Parameter Index 2=true
   )
   &self:ResolveRibbonHeadissuePre(New Simple Counter:NiagaraDataInterfaceSimpleCounter=$User.RibbonCounter2) {module=ResolveRibbonHeadissuePre}
  ]
  particleUpdateScript [
   &ParticleState()
   &ScaleColor(
    Scale Alpha=&FloatFromCurve(
     FloatCurve=DataInterface() {propertyValues {Curve {keys [
         {interpMode=RCIM_Cubic, leaveTangent=13.9561, tangentMode=RCTM_Break, time=0, value=0}
         {arriveTangent=2.4263, interpMode=RCIM_Cubic, tangentMode=RCTM_Break, time=0.1, value=1}
         {arriveTangent=0.265, interpMode=RCIM_Cubic, tangentMode=RCTM_Break, time=1, value=0}
        ]}}}
     Scale Curve=$System.DistanceNearFade
    )
    ScaleA=true
    ScaleRGB=true
   )
   &ScaleRibbonWidth(Ribbon Width Scale=&FloatFromCurve(FloatCurve=DataInterface() {propertyValues {Curve {keys [<at 0 1>,<at 1 0.35>]}}}))
   &ScaleRibbonWidth(Ribbon Width Scale=&ClampFloat(Float=&FN_LOD_Distance_Remap(Start Distance=0, End Distance=2000, Minimum Output=0.5, Maximum Output=1))) {module=ScaleRibbonWidth001}
   &SkeletalMeshLocation(
    Bone / Socket Index=$Particles.SpawnGroup
    Bone Sampling Mode=::NewEnumerator4
    MaskSpawnGroup=false
    Orientation Sampling=::NewEnumerator1
    Position Sampling=::NewEnumerator2
    Skeletal Mesh=DataInterface() {propertyValues {
      DefaultMesh=null
      ExcludeBoneName=root
      OptionalComponentClass=null
      PreviewMesh=&M_Med_Soldier_04_Preview
      SkinningMode=SkinOnTheFly
      SoftSourceActor=null
      SourceMode=Default
      UvSetIndex=0
      WholeMeshLOD=-1
      bExcludeBone=false
      bReadDeformedGeometry=true
      bRequireCurrentFrameData=true
      ComponentTags []
      FilteredBones [hand_l,hand_l]
      FilteredSockets []
      MeshUserParameter {parameter {name=None, typeDefHandle {registeredTypeIndex=1}, varData []}}
      SamplingRegions []
     }}
   ) {module=SkeletalMeshLocation001}
   &self:ResolveRibbonheadIssue(
    IDCustom:NiagaraInt32=$Particles.IDCustom
    InputHeadPos:NiagaraPosition=$Particles.SkeletalMeshLocation001.SampledPosition
    New Simple Counter:NiagaraDataInterfaceSimpleCounter=$User.RibbonCounter2
   ) {module=ResolveRibbonheadIssue}
  ]
  rendererClasses [&NiagaraRibbonRendererProperties]
  renderers [{rendererClass=&NiagaraRibbonRendererProperties}]
 }
 SandWarp004 {
  emitterUpdateScript [
   &EmitterState(
    Inactive Response=::NewEnumerator1
    Life Cycle Mode=::NewEnumerator1
    Loop Behavior=::NewEnumerator1
    Loop Duration Mode=::NewEnumerator1
    UseLoopDelay=false
   )
   &SpawnRate(
    SpawnRate=&MakeCustomFloatFromBool(
     Bool=$User.bIsAimAttacking
     True Float=&MakeCustomFloatFromBool(
      Bool=&SetBoolByFloatComparison(A=&VectorLength(Vector=$Engine.Owner.Velocity), B=1000, Comparison Type=::NewEnumerator6)
      True Float=&MakeCustomFloatFromBool(Bool=$User.bIsAtk, True Float=0, False Float=10)
      False Float=0
     )
     False Float=0
    )
    Use Spawn Probability=false
   )
   &SpawnBurst_Instantaneous(
    Spawn Count=&RandomRangeInt(Minimum=2, Maximum=2)
    Use Spawn Probability=false
    UseLoopCountLimit=false
   ) {enabled=false}
  ]
  particleSpawnScript [
   &InitializeParticle(
    Color Maximum=<rgba 1.6 1.7333 2 0.125>
    Color Minimum=<rgba 1.4 1.5167 1.75 0.05>
    Color Mode=::NewEnumerator2
    Lifetime Max=0.5
    Lifetime Min=0.35
    Lifetime Mode=::NewEnumerator1
    Mesh Scale Mode=::NewEnumerator1
    Mesh Uniform Scale Max=0.4
    Mesh Uniform Scale Min=0.3
    Position Offset=<0 0 50>
    UsePositionOffset=false
    Write Material Random=true
   )
   &SkeletalMeshLocation(
    Bone / Socket Index=19
    Bone Sampling Mode=::NewEnumerator4
    Interpolate Mesh Sampling=false
    MaskSpawnGroup=false
    Skeletal Mesh=$User.SkeletalMesh
   )
   &InitialMeshOrientation(
    Mesh Orientation Mode=::NewEnumerator7
    Rotation=&RandomRangeVector(Minimum=<0 -0.075 0>, Maximum=<0.1 0.125 0.2>)
    Use Rotation Vector=true
   )
   &AddVelocity(Velocity=<0 0 170>)
   /SetVariables(Particles.MeshOrientation=&MultiplyQuaternion(Quaternion A=$StackContext.SkeletalMeshLocation.SampledRotation, Quaternion B=$Particles.MeshOrientation)) {isSetParameters=true}
  ]
  particleUpdateScript [
   &ParticleState()
   &DynamicMaterialParameters(
    Index 0 Param 1=&FloatFromCurve(FloatCurve=DataInterface() {propertyValues {Curve {keys [{interpMode=RCIM_Cubic, tangentMode=RCTM_User, time=0.6, value=0}, {interpMode=RCIM_Cubic, tangentMode=RCTM_User, time=1, value=1}]}}})
    Index 0 Param 2=&MakeCustomFloatFromBool(
     Bool=$User.IsFrontEnd
     True Float=0
     False Float=&FNTimeOfDayScalar(Night Scale=0.05, Morning Scale=-0.26, Day Scale=0, Evening Scale=-0.35)
    )
    Index 0 Param 3=1
    Index 0 Param 4=&MakeCustomFloatFromBool(
     Bool=$User.IsFrontEnd
     True Float=0
     False Float=&FNTimeOfDayScalar(Night Scale=0.02, Morning Scale=-0.01, Day Scale=0, Evening Scale=-0.01)
    )
    Module.Param0WriteEnabled=false
    Module.Param1WriteEnabled=false
    Module.Param2WriteEnabled=true
    Module.Param3WriteEnabled=false
   )
   /SetVariables(
    Particles.XY:NiagaraFloat=&FloatFromCurve(
     FloatCurve=DataInterface() {propertyValues {Curve {keys [
         {arriveTangent=0.8311, interpMode=RCIM_Cubic, leaveTangent=0.8311, tangentMode=RCTM_User, time=0, value=0}
         {arriveTangent=0.4623, interpMode=RCIM_Cubic, leaveTangent=0.4623, tangentMode=RCTM_User, time=1, value=1}
        ]}}}
     Scale Curve=4
    )
    Particles.Z:NiagaraFloat=&FloatFromCurve(
     FloatCurve=DataInterface() {propertyValues {Curve {keys [
         {arriveTangent=3.7114, interpMode=RCIM_Cubic, leaveTangent=3.7114, tangentMode=RCTM_User, time=0, value=0}
         {arriveTangent=0.2081, interpMode=RCIM_Cubic, leaveTangent=0.2081, tangentMode=RCTM_User, time=0.4, value=0.525}
         {arriveTangent=-1.1305, interpMode=RCIM_Cubic, leaveTangent=-1.1305, tangentMode=RCTM_User, time=1, value=0}
        ]}}}
     Scale Curve=7
    )
   ) {isSetParameters=true}
   &ScaleMeshSize(Scale Factor=&MakeVector(X=$Particles.XY, Y=$Particles.XY, Z=$Particles.Z))
   &UpdateMeshOrientation(Rotation Rate=2.5, Rotation Vector=<0 0 1>)
   &DynamicMaterialParameters(
    Index 0 Param 1=&FloatFromCurve(FloatCurve=DataInterface() {propertyValues {Curve {keys [{interpMode=RCIM_Cubic, tangentMode=RCTM_User, time=0, value=1}, {interpMode=RCIM_Cubic, tangentMode=RCTM_User, time=1, value=0}]}}})
    Module.Param0WriteEnabled=true
    Module.Param1WriteEnabled=true
    Module.Param2WriteEnabled=true
    Module.Param3WriteEnabled=true
   ) {enabled=false, module=DynamicMaterialParameters001}
   &ScaleColor(
    Scale Alpha=&FloatFromCurve(
     FloatCurve=DataInterface() {propertyValues {Curve {keys [
         {arriveTangent=-0.1682, interpMode=RCIM_Cubic, leaveTangent=-0.1682, tangentMode=RCTM_User, time=0, value=0}
         {arriveTangent=2.5462, interpMode=RCIM_Cubic, leaveTangent=2.5462, tangentMode=RCTM_User, time=0.5, value=0.6}
         {arriveTangent=-2.3616, interpMode=RCIM_Cubic, leaveTangent=-2.3616, tangentMode=RCTM_User, time=0.8, value=0.9}
         {arriveTangent=-1.0334, interpMode=RCIM_Cubic, leaveTangent=-1.0334, tangentMode=RCTM_User, time=1, value=0}
        ]}}}
     Scale Curve=0.7
    )
    Scale RGB=&VectorFromFloat(Value=&FloatFromCurve(FloatCurve=DataInterface() {propertyValues {Curve {keys [{arriveTangent=-1.9639, time=0, value=1}, {arriveTangent=0.35, leaveTangent=0.35, time=1, value=1}]}}}))
    ScaleA=true
    ScaleRGB=true
   )
   /SetVariables(Particles.New Linear Color:LinearColor=&MultiplyLinearColor(Linear Color=$Particles.Color, Scale Factor=<0.8 0.9 1 0.75>)) {isSetParameters=true}
   &SolveForcesAndVelocity(Clamp Velocity=false, Limit Acceleration=false)
   /SetVariables(Particles.New Vector:Vector3f=&Multiply_Vector(A=$Particles.Scale, B=<1 1 0.6>)) {isSetParameters=true}
   &ScaleColor(
    Scale Alpha=&MakeCustomFloatFromBool(
     Bool=&SetBoolByFloatComparison(A=&VectorLength(Vector=$Engine.Owner.Velocity), B=1500, Comparison Type=::NewEnumerator6)
     True Float=1
     False Float=0
    )
    ScaleA=true
    ScaleRGB=true
   ) {module=ScaleColor001}
  ]
  rendererClasses [&NiagaraMeshRendererProperties]
  renderers [
   {rendererClass=&NiagaraMeshRendererProperties, rendererIndex=0}
   {rendererClass=&NiagaraMeshRendererProperties, rendererIndex=1}
  ]
 }
 Steam {
  emitterUpdateScript [
   &EmitterState(Life Cycle Mode=::NewEnumerator1, Loop Behavior=::NewEnumerator1, Loop Duration Mode=::NewEnumerator1, UseLoopDelay=true)
   &SpawnRate(SpawnRate=10, Use Spawn Probability=false)
  ]
  particleSpawnScript [
   &InitializeParticle(
    Color=<rgba 1.16 1.2567 1.45 1>
    Position Offset=<0 0 20>
    Position Offset Coordinate Space=::World
    Ribbon Width=100
    Ribbon Width Mode=::NewEnumerator1
    UsePositionOffset=true
    Write Material Random=true
   )
   /SetVariables(Particles.New float:NiagaraFloat=$Emitter.Age) {isSetParameters=true}
   &self:ResolveRibbonHeadissuePre(New Simple Counter:NiagaraDataInterfaceSimpleCounter=$User.RibbonParticleCounter) {module=ResolveRibbonHeadissuePre}
  ]
  particleUpdateScript [
   &ParticleState()
   &ScaleColor(
    Scale Alpha=&FloatFromCurve(
     FloatCurve=DataInterface() {propertyValues {Curve {keys [
         {interpMode=RCIM_Cubic, leaveTangent=0.3789, tangentMode=RCTM_Break, time=0.07, value=0}
         {arriveTangent=-3.0273, interpMode=RCIM_Cubic, leaveTangent=-3.0273, tangentMode=RCTM_User, time=0.27, value=1}
         {arriveTangent=-0.0831, interpMode=RCIM_Cubic, tangentMode=RCTM_Break, time=1, value=0}
        ]}}}
     Scale Curve=&RandomRangeFloat(Minimum=0.8, Maximum=0.8)
    )
    Scale RGB=&VectorFromFloat(Value=&FloatFromCurve(
      FloatCurve=DataInterface() {propertyValues {Curve {keys [{interpMode=RCIM_Cubic, tangentMode=RCTM_Break, time=0.04, value=0}, <at 0.3 1>]}}}
      CurveIndex=$Emitter.Age
     ))
    ScaleA=true
    ScaleRGB=true
   )
   &ScaleRibbonWidth(Ribbon Width Scale=&FloatFromCurve(FloatCurve=DataInterface() {propertyValues {Curve {keys [
         {arriveTangent=0.783, interpMode=RCIM_Cubic, leaveTangent=0.783, tangentMode=RCTM_User, time=0.2, value=0.61}
         {arriveTangent=0.3769, interpMode=RCIM_Cubic, leaveTangent=0.3769, tangentMode=RCTM_User, time=1, value=1}
        ]}}}))
   &DynamicMaterialParameters(
    Index 0 Param 1=&FloatFromCurve(FloatCurve=DataInterface() {propertyValues {Curve {keys [{interpMode=RCIM_Cubic, tangentMode=RCTM_Break, time=0.5, value=1}, <at 1 0>]}}})
    Module.Param0WriteEnabled=true
    Module.Param1WriteEnabled=true
    Module.Param2WriteEnabled=true
    Module.Param3WriteEnabled=true
   )
   &ScaleColor(
    Scale Alpha=&FloatFromCurve(
     FloatCurve=DataInterface() {propertyValues {Curve {keys [{interpMode=RCIM_Cubic, tangentMode=RCTM_Break, time=0.1, value=0}, <at 1 1>]}}}
     CurveIndex=$Particles.New float
     Scale Curve=$User.GlobalOpacity
    )
    Scale RGB=<0.35 0.35 0.35>
    ScaleA=true
    ScaleRGB=true
   ) {module=ScaleColor001}
   &ScaleRibbonWidth(Ribbon Width Scale=&FloatFromCurve(FloatCurve=DataInterface() {propertyValues {Curve {keys [<at 0 1>,<at 0.6 0.75>]}}}, CurveIndex=$System.Age)) {module=ScaleRibbonWidth001}
   &self:ResolveRibbonheadIssue(
    IDCustom:NiagaraInt32=$Particles.IDCustom
    InputHeadPos:NiagaraPosition=$Engine.Owner.Position
    New Simple Counter:NiagaraDataInterfaceSimpleCounter=$User.RibbonParticleCounter
   ) {module=ResolveRibbonheadIssue}
  ]
  rendererClasses [&NiagaraRibbonRendererProperties]
  renderers [{rendererClass=&NiagaraRibbonRendererProperties}]
 }
}
summary {
 systemName=NS_GasParticles_NyxGlass
 emitters [
  {emitterName=Steam, rendererClasses [&NiagaraRibbonRendererProperties]}
  {emitterName=HandTrail, rendererClasses [&NiagaraRibbonRendererProperties]}
  {emitterName=HandTrail001, rendererClasses [&NiagaraRibbonRendererProperties]}
  {emitterName=GrindRail_Main, rendererClasses [&NiagaraRibbonRendererProperties]}
  {emitterName=SandWarp004, rendererClasses [&NiagaraMeshRendererProperties]}
 ]
 userVariables [
  User.ActorV {type=Vector3f, description="", defaultValue {struct=Vector3f, value=<0 0 0>}}
  User.bIsAimAttacking {type=NiagaraBool}
  User.bIsAtk {type=NiagaraBool}
  User.bIsFrontEnd {type=NiagaraBool}
  User.bIsFrontEndPreview {type=NiagaraBool}
  User.Delay {type=NiagaraFloat, defaultValue=0, description=""}
  User.GlobalOpacity {type=NiagaraFloat, defaultValue=0.5, description=""}
  User.IsFrontEnd {type=NiagaraBool}
  User.RibbonCounter1 {
   type=NiagaraDataInterfaceSimpleCounter
   description=""
   defaultValue {
    struct=NiagaraExt_VariableValue_DataInterface
    value {
     dataInterface=&self:NiagaraDataInterfaceSimpleCounter_2
     dataInterfaceClass=/Script/Niagara.NiagaraDataInterfaceSimpleCounter
    }
   }
  }
  User.RibbonCounter2 {
   type=NiagaraDataInterfaceSimpleCounter
   description=""
   defaultValue {
    struct=NiagaraExt_VariableValue_DataInterface
    value {
     dataInterface=&self:NiagaraDataInterfaceSimpleCounter_1
     dataInterfaceClass=&NiagaraDataInterfaceSimpleCounter
    }
   }
  }
  User.RibbonParticleCounter {
   type=NiagaraDataInterfaceSimpleCounter
   description=""
   defaultValue {
    struct=NiagaraExt_VariableValue_DataInterface
    value {
     dataInterface=&self:NiagaraDataInterfaceSimpleCounter_0
     dataInterfaceClass=&NiagaraDataInterfaceSimpleCounter
    }
   }
  }
  User.Scale {type=NiagaraFloat, defaultValue=1, description=""}
  User.SkeletalMesh {
   type=NiagaraDataInterfaceSkeletalMesh
   defaultValue {
    struct=NiagaraExt_VariableValue_DataInterface
    value {dataInterface=&self:NiagaraDataInterfaceSkeletalMesh_0}
   }
  }
  User.SprayDirection {type=Vector3f, description="", defaultValue {struct=Vector3f, value=<0 0 0>}}
  User.SpraySpeed {type=NiagaraFloat, defaultValue=3000, description=""}
  User.TempVF {type=NiagaraFloat, defaultValue=5000, description=""}
  User.TempVFCur {
   type=NiagaraDataInterfaceCurve
   description=""
   defaultValue {
    struct=NiagaraExt_VariableValue_DataInterface
    value {
     dataInterface=&self:NiagaraDataInterfaceCurve_0
     dataInterfaceClass=&NiagaraDataInterfaceCurve
    }
   }
  }
  User.TurnON_Niagara_Velocity {type=NiagaraBool}
 ]
}
