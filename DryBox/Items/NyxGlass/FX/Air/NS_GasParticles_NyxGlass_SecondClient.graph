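# Niagara air effect for the NyxGlass item's gas, named as a second-client variant, drawn by one Steam ribbon emitter.
# Bluish-white particles spawn continuously, with color, opacity and width shaped by particle, emitter and system age curves and User.GlobalOpacity.
# Custom ResolveRibbonHead scratch modules use a shared counter, custom IDs and owner position, apparently to correct the ribbon head.
# Most other user parameters, such as spray settings, are unreferenced in the shown stack.
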NiagaraSystem /DryBox/Items/NyxGlass/FX/Air/NS_GasParticles_NyxGlass_SecondClient
dependencies {
 usedDataInterfaces []
 usedDynamicInputs [
  &Lerp_Float
  &ClampFloat
  &FN_LOD_Distance_Remap
 ]
 usedModules [
  &SystemState
  &self:SystemSpawnScript.SystemScriptSource.SystemScriptGraph.NiagaraNodeAssignment_0.SetVariables_#1
 ]
}
emitters {Steam {
  emitterUpdateScript [
   &EmitterState(Life Cycle Mode=::NewEnumerator1, Loop Behavior=::NewEnumerator1, Loop Duration Mode=::NewEnumerator1, UseLoopDelay=true)
   &SpawnRate(SpawnRate=10, Use Spawn Probability=false)
  ]
  particleSpawnScript [
   &InitializeParticle(
    Color=<rgba 1.16 1.2567 1.45 1>
    Position Offset=<0 0 20>
    Position Offset Coordinate Space=::World
    Ribbon Width=100
    Ribbon Width Mode=::NewEnumerator1
    UsePositionOffset=true
    Write Material Random=true
   )
   /SetVariables(Particles.New float:NiagaraFloat=$Emitter.Age) {isSetParameters=true}
   &self:ResolveRibbonHeadissuePre(New Simple Counter:NiagaraDataInterfaceSimpleCounter=$User.RibbonParticleCounter) {module=ResolveRibbonHeadissuePre}
  ]
  particleUpdateScript [
   &ParticleState()
   &ScaleColor(
    Scale Alpha=&FloatFromCurve(
     FloatCurve=DataInterface() {propertyValues {Curve {keys [
         {interpMode=RCIM_Cubic, leaveTangent=0.3789, tangentMode=RCTM_Break, time=0.07, value=0}
         {arriveTangent=-3.0273, interpMode=RCIM_Cubic, leaveTangent=-3.0273, tangentMode=RCTM_User, time=0.27, value=1}
         {arriveTangent=-0.0831, interpMode=RCIM_Cubic, tangentMode=RCTM_Break, time=1, value=0}
        ]}}}
     Scale Curve=&RandomRangeFloat(Minimum=0.8, Maximum=0.8)
    )
    Scale RGB=&VectorFromFloat(Value=&FloatFromCurve(
      FloatCurve=DataInterface() {propertyValues {Curve {keys [{interpMode=RCIM_Cubic, tangentMode=RCTM_Break, time=0.04, value=0}, <at 0.3 1>]}}}
      CurveIndex=$Emitter.Age
     ))
    ScaleA=true
    ScaleRGB=true
   )
   &ScaleRibbonWidth(Ribbon Width Scale=&FloatFromCurve(FloatCurve=DataInterface() {propertyValues {Curve {keys [
         {arriveTangent=0.783, interpMode=RCIM_Cubic, leaveTangent=0.783, tangentMode=RCTM_User, time=0.2, value=0.61}
         {arriveTangent=0.3769, interpMode=RCIM_Cubic, leaveTangent=0.3769, tangentMode=RCTM_User, time=1, value=1}
        ]}}}))
   &DynamicMaterialParameters(
    Index 0 Param 1=&FloatFromCurve(FloatCurve=DataInterface() {propertyValues {Curve {keys [{interpMode=RCIM_Cubic, tangentMode=RCTM_Break, time=0.5, value=1}, <at 1 0>]}}})
    Module.Param0WriteEnabled=true
    Module.Param1WriteEnabled=true
    Module.Param2WriteEnabled=true
    Module.Param3WriteEnabled=true
   )
   &ScaleColor(
    Scale Alpha=&FloatFromCurve(
     FloatCurve=DataInterface() {propertyValues {Curve {keys [{interpMode=RCIM_Cubic, tangentMode=RCTM_Break, time=0.1, value=0}, <at 1 1>]}}}
     CurveIndex=$Particles.New float
     Scale Curve=$User.GlobalOpacity
    )
    Scale RGB=<0.35 0.35 0.35>
    ScaleA=true
    ScaleRGB=true
   ) {module=ScaleColor001}
   &ScaleRibbonWidth(Ribbon Width Scale=&FloatFromCurve(FloatCurve=DataInterface() {propertyValues {Curve {keys [<at 0 1>,<at 0.6 0.75>]}}}, CurveIndex=$System.Age)) {module=ScaleRibbonWidth001}
   &self:ResolveRibbonheadIssue(
    IDCustom:NiagaraInt32=$Particles.IDCustom
    InputHeadPos:NiagaraPosition=$Engine.Owner.Position
    New Simple Counter:NiagaraDataInterfaceSimpleCounter=$User.RibbonParticleCounter
   ) {module=ResolveRibbonheadIssue}
  ]
  rendererClasses [&NiagaraRibbonRendererProperties]
  renderers [{rendererClass=&NiagaraRibbonRendererProperties}]
 }}
summary {
 systemName=NS_GasParticles_NyxGlass_SecondClient
 emitters [{emitterName=Steam, rendererClasses [&NiagaraRibbonRendererProperties]}]
 userVariables [
  User.ActorV {type=Vector3f, description="", defaultValue {struct=Vector3f, value=<0 0 0>}}
  User.bIsAimAttacking {type=NiagaraBool}
  User.bIsFrontEnd {type=NiagaraBool}
  User.bIsFrontEndPreview {type=NiagaraBool}
  User.Delay {type=NiagaraFloat, defaultValue=0, description=""}
  User.GlobalOpacity {type=NiagaraFloat, defaultValue=0.5, description=""}
  User.IsFrontEnd {type=NiagaraBool}
  User.RibbonCounter1 {
   type=NiagaraDataInterfaceSimpleCounter
   description=""
   defaultValue {
    struct=NiagaraExt_VariableValue_DataInterface
    value {
     dataInterface=&self:NiagaraDataInterfaceSimpleCounter_2
     dataInterfaceClass=/Script/Niagara.NiagaraDataInterfaceSimpleCounter
    }
   }
  }
  User.RibbonCounter2 {
   type=NiagaraDataInterfaceSimpleCounter
   description=""
   defaultValue {
    struct=NiagaraExt_VariableValue_DataInterface
    value {
     dataInterface=&self:NiagaraDataInterfaceSimpleCounter_1
     dataInterfaceClass=&NiagaraDataInterfaceSimpleCounter
    }
   }
  }
  User.RibbonParticleCounter {
   type=NiagaraDataInterfaceSimpleCounter
   description=""
   defaultValue {
    struct=NiagaraExt_VariableValue_DataInterface
    value {
     dataInterface=&self:NiagaraDataInterfaceSimpleCounter_0
     dataInterfaceClass=&NiagaraDataInterfaceSimpleCounter
    }
   }
  }
  User.Scale {type=NiagaraFloat, defaultValue=1, description=""}
  User.SkeletalMesh {
   type=NiagaraDataInterfaceSkeletalMesh
   defaultValue {
    struct=NiagaraExt_VariableValue_DataInterface
    value {dataInterface=&self:NiagaraDataInterfaceSkeletalMesh_0}
   }
  }
  User.SprayDirection {type=Vector3f, description="", defaultValue {struct=Vector3f, value=<0 0 0>}}
  User.SpraySpeed {type=NiagaraFloat, defaultValue=3000, description=""}
  User.TempVF {type=NiagaraFloat, defaultValue=5000, description=""}
  User.TempVFCur {
   type=NiagaraDataInterfaceCurve
   description=""
   defaultValue {
    struct=NiagaraExt_VariableValue_DataInterface
    value {
     dataInterface=&self:NiagaraDataInterfaceCurve_0
     dataInterfaceClass=&NiagaraDataInterfaceCurve
    }
   }
  }
  User.TurnON_Niagara_Velocity {type=NiagaraBool}
 ]
}
